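# Translucent unlit master material for display cases, faking a lit, smoky interior behind reflective glass.
# Distorted scrolling smoke, virtual-plane UVs, depth masks and a local-space fake spotlight build color and opacity.
# Lat-long reflection lookup, Phong, specular, fresnel, edge masks and custom-depth/world-normal highlights add detail, with inverse eye adaptation on the result.
# An "Opaque?" static switch bypasses translucency-dependent effects; low/mobile switches drop or simplify costly terms.
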Material /Hera/Props/GamePlay/TechArt/M_Displaycase_Master
graph {
 properties {
  bAutomaticallySetUsageInEditor=false
  bSubstrateRoughnessTracking=false
  bUseTranslucencyVertexFog=false
  blendMode=BLEND_Translucent
  shadingModel=MSM_Unlit
  translucencyLightingMode=TLM_VolumetricDirectional
  usedShadingModels=Unlit
 }
 nodes [
  n1 = NamedRerouteUsage()
  n2 = NamedRerouteUsage()
  n3 = NamedRerouteUsage()
  n4 = NamedRerouteUsage()
  n5 = Multiply(A=Multiply(A=Add(A=@n3, B=@n4), B=ScalarParameter() {defaultValue=1, parameterName=Emissive Overall Intensity}), B=NamedRerouteUsage())
  n6 = Add(
   A=LinearInterpolate(
    A=@n2
    B=Multiply(A=@n2, B=@n5)
    Alpha=ScalarParameter() {defaultValue=0.5, group=Fake Reflection, parameterName=Reflection Color Blend, sliderMax=1}
   ) {constAlpha=0.3}
   B=Add(A=@n5, B=NamedRerouteUsage())
  )
  n7 = NamedRerouteUsage()
  EyeAdaptationInverse(
   LightValueInput=StaticSwitchParameter(True=Multiply(A=@n1, B=@n6), False=Add(A=@n6, B=@n7)) {parameterName=Opaque?}
   AlphaInput=Constant() {r=0.85}
  )
  n8 = DepthFade(FadeDistance=Constant() {r=80})
  n9 = Multiply(A=@n8, B=@n8)
  MaterialFunctionCall!(Default=@n9, Low/Mobile=Constant()) {materialFunction=/Game/Effects/Fort_Effects/Materials/_Functions/MF_LowMobile_PathingSwitch}
  n10 = CameraVectorWS()
  n11 = Add(A=ObjectPositionWS(), B=ScalarParameter() {parameterName=Plane Height})
  n12 = ScalarParameter() {defaultValue=250, parameterName=Virtual plane Scale}
  NamedRerouteDeclaration(None=StaticSwitchParameter(
    True=Constant()
    False=MaterialFunctionCall(Plane Normal Axis=@n10, Plane Center=@n11, UV Scale=@n12) {materialFunction=&VirtualPlaneCoordinates}.UVs
   ) {parameterName=Opaque?}) {name=Virtual Plane UVs, nodeColor=<rgba 0 1 0.8828 1>}
  n13 = VertexNormalWS()
  n14 = Abs(None=ComponentMask(None=Transform(None=@n13) {transformSourceType=TRANSFORMSOURCE_World, transformType=TRANSFORM_Local}) {r=true})
  n15 = Divide(
   A=Subtract(A=SceneDepth().R, B=PixelDepth().R)
   B=LinearInterpolate(
    A=ScalarParameter() {defaultValue=295.5761, group=Depth, parameterName=Depth Distance Y}
    B=ScalarParameter() {defaultValue=500, group=Depth, parameterName=Depth Distance X}
    Alpha=Saturate(None=@n14)
   )
  )
  n16 = ScalarParameter() {defaultValue=2.2081, group=Depth, parameterName=Depth Offset}
  n17 = OneMinus(None=Saturate(None=Add(A=@n15, B=@n16)))
  n18 = Multiply(A=@n17, B=@n17)
  NamedRerouteDeclaration(None=StaticSwitchParameter(True=Constant() {r=1}, False=OneMinus(None=@n18)) {parameterName=Opaque?}) {name=Final Depth Mask, nodeColor=<rgba 1 0 0.7266 1>}
  n19 = NamedRerouteUsage()
  n20 = PixelNormalWS()
  n21 = Multiply(
   A=Multiply(A=DotProduct(A=@n19, B=@n20), B=ScalarParameter() {defaultValue=1, group=mbient Light, parameterName=Phong Normal Mult}) {constB=2}
   B=@n20
  )
  n22 = DotProduct(
   A=Normalize(VectorInput=Subtract(A=@n21, B=@n19))
   B=Multiply(A=MaterialFunctionCall() {materialFunction=&CameraDirectionVector}.Result) {constB=-1}
  )
  n23 = Multiply(
   A=Power(Base=Max(A=@n22) {constB=0}, Exp=ScalarParameter() {defaultValue=15, group=mbient Light, parameterName=Fake Phong Contrast})
   B=ScalarParameter() {defaultValue=0.052, group=mbient Light, parameterName=Fake Phong Intensity}
  )
  NamedRerouteDeclaration(None=MaterialFunctionCall(Default=Clamp(None=@n23) {maxDefault=0.5}, Low/Mobile=Constant()) {materialFunction=&MF_LowMobile_PathingSwitch}.Result) {name=Phong Reflections, nodeColor=<rgba 1 0 0.6562 1>}
  n24 = Power(
   Base=DotProduct(A=CameraVectorWS(), B=VertexNormalWS())
   Exp=ScalarParameter() {defaultValue=0.2003, group=Fresnel Mask, parameterName=Fresnel Power}
  )
  n25 = OneMinus(None=@n24)
  n26 = ScalarParameter() {group=Fresnel Mask, parameterName=Invert Fresnel}
  NamedRerouteDeclaration(None=Multiply(A=LinearInterpolate(A=@n25, B=@n24, Alpha=@n26), B=ScalarParameter() {defaultValue=4, group=Fresnel Mask, parameterName=Fresnel Mult})) {name=Fresnel, nodeColor=<rgba 0 0.9297 1 1>}
  n27 = Normalize(VectorInput=Add(A=SkyAtmosphereLightDirection(), B=CameraVectorWS()))
  n28 = PixelNormalWS()
  n29 = Multiply(
   A=MaterialFunctionCall(Default=ScalarParameter() {defaultValue=0.7114}, Low/Mobile=DotProduct(A=@n27, B=@n28)) {materialFunction=&MF_LowMobile_PathingSwitch}.Result
   B=VectorParameter() {group=Specular, parameterName=Specular Color, defaultValue=<rgba 1 1 1 1>}.RGB
  )
  n30 = Power(Base=@n29, Exp=ScalarParameter() {defaultValue=104, group=Specular, parameterName=Specular Spot size})
  NamedRerouteDeclaration(None=Power(Base=@n30) {constExponent=4}) {name=Specular Spot, nodeColor=<rgba 0 0.25 1 1>}
  n31 = ScalarParameter() {bUseCustomPrimitiveData=true, defaultValue=1, parameterName=HueShift}
  n32 = Multiply(A=ScalarParameter() {defaultValue=1, group=Color, parameterName=Smoke Color Shift}, B=NamedRerouteUsage())
  n33 = VectorParameter() {group=Color, parameterName=Base Color, defaultValue=<rgba 0.3385 0.1258 0.0241 1>}
  n34 = ScalarParameter() {defaultValue=0.95, group=Color, parameterName=Smoke Color Shift Intensity}
  n35 = ScalarParameter() {defaultValue=0.8, group=Color, parameterName=Fresnel Color Shift Max}
  n36 = NamedRerouteUsage()
  n37 = Multiply(
   A=LinearInterpolate(
    A=MaterialFunctionCall(Hue Shift Percentage=@n32, Texture=@n33.RGB) {materialFunction=&HueShift}.Result
    B=@n33.RGB
    Alpha=LinearInterpolate(A=@n34, B=@n35, Alpha=@n36)
   ) {constAlpha=0.95}
   B=NamedRerouteUsage()
  )
  n38 = MaterialFunctionCall(Default=Constant() {r=1}, Low/Mobile=ScalarParameter() {defaultValue=1.5, parameterName=Low/Mobile Color Mult}) {materialFunction=&MF_LowMobile_PathingSwitch}
  NamedRerouteDeclaration(None=StaticSwitchParameter(
    True=MaterialFunctionCall(Hue Shift Percentage=@n31, Texture=@n37) {materialFunction=&HueShift}.Result
    False=Multiply(A=@n37, B=@n38.Result)
   ) {parameterName=Opaque?}) {name=Color, nodeColor=<rgba 1 0 0.9141 1>}
  n39 = SceneTexture() {sceneTextureId=PPI_WorldNormal}
  n40 = Multiply(A=ComponentMask(None=@n39.Color) {b=true}) {constB=-1}
  NamedRerouteDeclaration(None=StaticSwitchParameter(True=Constant(), False=Saturate(None=@n40)) {parameterName=Opaque?}) {name=Down Facing Normal, nodeColor=<rgba 1 0.6797 0 1>}
  NamedRerouteDeclaration(None=StaticSwitchParameter(True=Constant() {r=1}, False=@n9) {parameterName=Opaque?}) {name=Depth Fade, nodeColor=<rgba 0 0.8594 1 1>}
  n41 = OneMinus(None=Fresnel(ExponentIn=ScalarParameter() {defaultValue=1.5, parameterName=Fresnel 2}) {baseReflectFraction=0})
  n42 = Multiply(A=@n41, B=@n41)
  NamedRerouteDeclaration(None=Saturate(None=Add(A=@n42) {constB=0.3})) {name=Fresnel mask, nodeColor=<rgba 1 0 0.3516 1>}
  n43 = PixelDepth()
  n44 = SceneTexture() {sceneTextureId=PPI_CustomDepth}
  n45 = ComponentMask(None=Add(A=Subtract(A=@n43.R, B=@n44.Color)) {constB=200}) {r=true}
  NamedRerouteDeclaration(None=StaticSwitchParameter(True=Constant(), False=Saturate(None=@n45)) {parameterName=Opaque?}) {name=Custom Depth, nodeColor=<rgba 0 0.2734 1 1>}
  NamedRerouteDeclaration(None=TextureObjectParameter() {group=Smoke, parameterName=Smoke texture, texture=/Game/Items/Traps/Textures/T_Barrier_Clouds_03_M}) {name=Smoke Texture, nodeColor=<rgba 0.7578 0 1 1>}
  n46 = ScalarParameter() {defaultValue=1, group=Spotlight, parameterName=Spotlight Mult}
  n47 = VectorParameter() {group=Spotlight, parameterName=Light Offset, defaultValue=<rgba 0 9.3576 105 1>}
  n48 = Divide(
   A=Subtract(A=LocalPosition().XYZ, B=@n47.RGB)
   B=VectorParameter() {group=Spotlight, parameterName=Light Scale, defaultValue=<rgba 30.2842 21.061 34.1218 1>}.RGB
  )
  n49 = Add(
   A=@n47.RGB
   B=Transform(None=ActorPositionWS() {originType=CameraRelative}) {transformSourceType=TRANSFORMSOURCE_World, transformType=TRANSFORM_Local}
  )
  n50 = DotProduct(A=@n48, B=@n49)
  n51 = Transform(None=WorldPosition() {worldPositionShaderOffset=WPT_CameraRelative}.XYZ) {transformSourceType=TRANSFORMSOURCE_World, transformType=TRANSFORM_Local}
  n52 = DotProduct(A=@n49, B=@n51)
  n53 = Subtract(A=@n48, B=Multiply(A=Divide(A=@n50, B=@n52), B=@n51))
  MaterialExpressionPower_3 = Power(Base=Abs(None=Divide(A=@n53))) {constExponent=2.3}
  n54 = Convert(None=@MaterialExpressionPower_3) {
   nodeName=Break Float3
   convertInputs [{
     type=Vector3
     expressionInput {
      expression=&self:MaterialExpressionPower_3
      inputName=None
      mask=0
      maskA=0
      maskB=0
      maskG=0
      maskR=0
      outputIndex=0
     }
    }]
   convertOutputs [{type=Scalar}, {type=Scalar}, {type=Scalar}]
  }
  n55 = Add(A=@n54, B=@n54)
  n56 = OneMinus(None=Saturate(None=Add(A=@n55, B=@n54)))
  NamedRerouteDeclaration(None=StaticSwitchParameter(True=Constant(), False=Multiply(A=@n46, B=@n56)) {parameterName=Opaque?}) {name=Spotlight, nodeColor=<rgba 1 0 0.8906 1>}
  n57 = Add(
   A=VertexNormalWS()
   B=Multiply(A=ScalarParameter() {group=Fake Reflection, parameterName=Reflection Flatten Normal}, B=Constant3Vector() {constant=<rgba 0 0 1 0>})
  )
  NamedRerouteDeclaration(None=ReflectionVectorWS(CustomWorldNormal=Normalize(VectorInput=@n57))) {name=Reflection Vector With Normals, nodeColor=<rgba 1 0.2578 0 1>}
  n58 = Multiply(A=MaterialFunctionCall() {materialFunction=&Color_Time_of_Day}.Result, B=NamedRerouteUsage())
  n59 = NamedRerouteUsage()
  n60 = Desaturation(
   None=Saturate(None=LinearInterpolate(A=@n58, B=@n59))
   Fraction=ScalarParameter() {defaultValue=0.8, group=Ambient Light, parameterName=Ambient Light Desaturation}
  )
  n61 = MaterialFunctionCall(
   Default=VectorParameter() {group=Ambient Light, parameterName=Ambient Light Mult, defaultValue=<rgba 0 0 0 1>}.RGB
   Low/Mobile=VectorParameter() {group=Ambient Light, parameterName="Ambient Light Mult  Mobile", defaultValue=<rgba 0 0 0 1>}.RGB
  ) {materialFunction=&MF_LowMobile_PathingSwitch}
  n62 = NamedRerouteUsage()
  NamedRerouteDeclaration(None=Multiply(A=Multiply(A=@n60, B=@n61.Result), B=Add(A=@n62) {constB=0.2})) {name=Ambient Light, nodeColor=<rgba 1 0 0.2109 1>}
  n63 = NamedRerouteUsage()
  n64 = VectorParameter() {group=Up HighLights, parameterName=Color Spotlight Fade, defaultValue=<rgba 0.3958 0.2936 0.0887 1>}
  n65 = ScalarParameter() {defaultValue=0.953, group=Up HighLights, parameterName=Hightlight Light Aura}
  n66 = ScalarParameter() {defaultValue=1.1, group=Up HighLights, parameterName=Hightlight Up Light}
  n67 = NamedRerouteUsage()
  n68 = Multiply(A=@n67, B=NamedRerouteUsage())
  n69 = Add(A=NamedRerouteUsage(), B=@n67)
  n70 = NamedRerouteUsage()
  n71 = Add(
   A=Multiply(A=NamedRerouteUsage(), B=LinearInterpolate(A=@n65, B=@n66, Alpha=@n68))
   B=LinearInterpolate(
    A=ScalarParameter() {group=Up HighLights, parameterName=Custom Mask - Aura Bias}
    B=ScalarParameter() {defaultValue=0.25, group=Up HighLights, parameterName=Custom Mask - Up Light Bias}
    Alpha=Multiply(A=@n69, B=@n70)
   )
  )
  n72 = StaticSwitchParameter(True=Constant(), False=Saturate(None=Power(Base=@n71) {constExponent=4})) {parameterName=Opaque?}
  n73 = MaterialFunctionCall(
   Default=LinearInterpolate(
    A=LinearInterpolate(A=@n63, B=@n64.RGB, Alpha=@n72)
    B=VectorParameter() {group=Up HighLights, parameterName=Color Highlight, defaultValue=<rgba 1 0.4979 0.1458 1>}.RGB
    Alpha=Power(Base=@n72) {constExponent=4}
   )
   Low/Mobile=NamedRerouteUsage()
  ) {materialFunction=&MF_LowMobile_PathingSwitch}
  n74 = Multiply(
   A=LinearInterpolate(
    A=ScalarParameter() {defaultValue=1, group=Color, parameterName=Front Ground Color Mult}
    B=ScalarParameter() {defaultValue=1, group=Color, parameterName=Back Ground Color Mult}
    Alpha=NamedRerouteUsage()
   )
   B=Multiply(A=NamedRerouteDeclaration(None=@n73.Result) {name=Final Gradings, nodeColor=<rgba 0 0.25 1 1>}, B=NamedRerouteUsage())
  )
  n75 = NamedRerouteUsage()
  n76 = Multiply(A=NamedRerouteUsage(), B=VectorParameter() {group=Edge Masks, parameterName=Border color, defaultValue=<rgba 1 1 1 1>}.RGB)
  n77 = NamedRerouteUsage()
  NamedRerouteDeclaration(None=Add(A=Multiply(A=@n74, B=@n75), B=Multiply(A=@n76, B=@n77))) {name=Final color, nodeColor=<rgba 0.0312 1 0 1>}
  n78 = Divide(
   A=Add(A=LocalPosition().Z, B=ScalarParameter() {group=Pos GRad, parameterName=Pos Gradient Offset})
   B=ScalarParameter() {defaultValue=1, group=Pos GRad, parameterName=Pos Gradient Size}
  )
  NamedRerouteDeclaration(None=LinearInterpolate(B=VectorParameter() {group=Pos GRad, parameterName=Pos Gradient, defaultValue=<rgba 1 1 1 1>}.RGB, Alpha=Saturate(None=@n78)) {constA=1}) {name=Pos Grad, nodeColor=<rgba 0 0.2266 1 1>}
  n79 = Constant() {r=1}
  n80 = Divide(A=Time(), B=ScalarParameter() {defaultValue=50, group=Smoke, parameterName=Smoke Speed})
  n81 = Add(
   A=NamedRerouteUsage()
   B=Multiply(A=NamedRerouteDeclaration(None=@n80) {name=Smk Time, nodeColor=<rgba 1 0.4922 0 1>}, B=Constant2Vector() {g=0.83, r=0.7})
  )
  n82 = Multiply(A=@n81, B=ScalarParameter() {defaultValue=1, group=Smoke, parameterName=Smoke Scale 2})
  n83 = ScalarParameter() {defaultValue=0.1, group=Smoke, parameterName=Smoke Distortion Intensity}
  n84 = NamedRerouteUsage()
  n85 = ScalarParameter() {defaultValue=0.432, group=Smoke, parameterName=Smoke Scale}
  n86 = NamedRerouteUsage()
  n87 = Constant2Vector() {g=-0.64, r=-0.8}
  n88 = TextureSample(UVs=Add(A=Multiply(A=@n84, B=@n85), B=Multiply(A=@n86, B=@n87))) {samplerSource=SSM_Wrap_WorldGroupSettings, texture=/Game/Athena/SafeZone/PulseLancerAssets/T_StarSky_SafeZone}
  n89 = NamedRerouteUsage()
  n90 = TextureSample(UVs=Add(A=@n82, B=Multiply(A=@n88.R, B=@n83) {constB=0.2}), Tex=@n89) {samplerSource=SSM_Wrap_WorldGroupSettings, texture=&T_StarSky_SafeZone}
  n91 = Multiply(A=@n90.R, B=@n90.R)
  n92 = Add(A=Multiply(A=@n81) {constB=2.5}, B=ComponentMask(None=Multiply(A=@n83, B=@n88.RGB) {constB=0.2}) {r=true})
  n93 = NamedRerouteUsage()
  n94 = SmoothStep(
   Min=ScalarParameter() {defaultValue=-0.004, group=Smoke, parameterName=Smoke Contrast Min}
   Max=ScalarParameter() {defaultValue=0.39, group=Smoke, parameterName=Smoke Contrast Max}
   Value=Multiply(
    A=Multiply(A=@n91, B=@n88.B) {constA=1}
    B=TextureSample(UVs=@n92, Tex=@n93) {samplerSource=SSM_Wrap_WorldGroupSettings, texture=&T_StarSky_SafeZone}.B
   )
  )
  n95 = ScalarParameter() {defaultValue=0.16, group=Smoke, parameterName=Smoke Power}
  n96 = StaticSwitchParameter(
   True=@n79
   False=MaterialFunctionCall(
    Default=Power(Base=@n94, Exp=@n95)
    Low/Mobile=TextureSample(UVs=@n82, Tex=@n89) {samplerSource=SSM_Wrap_WorldGroupSettings, texture=&T_StarSky_SafeZone}.G
   ) {materialFunction=&MF_LowMobile_PathingSwitch}.Result
  ) {parameterName=Opaque?}
  n97 = CameraVectorWS()
  n98 = DotProduct(A=ComponentMask(None=Saturate(None=@n97)) {b=true}, B=ComponentMask(None=VertexNormalWS()) {b=true})
  n99 = Saturate(None=Multiply(A=@n98, B=@n98))
  n100 = Constant()
  n101 = Saturate(None=Add(
    A=ScalarParameter() {defaultValue=0.2, group=Smoke, parameterName=Smoke Softness}
    B=MaterialFunctionCall(Default=@n99, Low/Mobile=@n100) {materialFunction=&MF_LowMobile_PathingSwitch}.Result
   ))
  NamedRerouteDeclaration(None=Saturate(None=LinearInterpolate(A=@n96, B=@n79, Alpha=@n101))) {name=Smoke, nodeColor=<rgba 1 0 0.2812 1>}
  n102 = CameraVectorWS()
  n103 = Multiply(
   A=Power(
    Base=DotProduct(A=@n102, B=@n39.Color)
    Exp=ScalarParameter() {defaultValue=5, group=Up HighLights, parameterName=Highlight Fresnel Contrast}
   )
   B=ScalarParameter() {defaultValue=0.5, group=Up HighLights, parameterName=Highlight Fresnel Intensity}
  )
  NamedRerouteDeclaration(None=MaterialFunctionCall(Default=Saturate(None=@n103), Low/Mobile=Constant()) {materialFunction=&MF_LowMobile_PathingSwitch}.Result) {name=Fresnel Higlight, nodeColor=<rgba 0.9922 0 1 1>}
  n104 = Multiply(A=@n29, B=ScalarParameter() {defaultValue=0.6221, group=Specular, parameterName=Specular Darknenning})
  NamedRerouteDeclaration(None=Add(A=@n30, B=OneMinus(None=@n104))) {name=Specular Highlight, nodeColor=<rgba 1 0 0.1641 1>}
  n105 = TextureSampleParameter2D() {group=Edge Masks, parameterName=Edge Mask Tex, samplerSource=SSM_Wrap_WorldGroupSettings, texture=&/T_Sides}
  n106 = Multiply(A=@n105.G, B=ScalarParameter() {group=Edge Masks, parameterName=Edge Mask Mult})
  NamedRerouteDeclaration(None=Multiply(A=@n106, B=@n106)) {name=Edges Mask, nodeColor=<rgba 0.3359 1 0 1>}
  n107 = Constant()
  n108 = VectorParameter() {group=Fake Reflection, parameterName=Reflection Color, defaultValue=<rgba 1 1 1 1>}
  n109 = NamedRerouteUsage()
  n110 = Arctangent2Fast(Y=ComponentMask(None=@n109) {g=true}, X=ComponentMask(None=@n109) {r=true})
  n111 = MaterialFunctionCall(Multiplier=Constant() {r=2}) {materialFunction=&Pi}
  n112 = Multiply(
   A=Frac(None=Divide(A=@n110, B=@n111.Result))
   B=ScalarParameter() {defaultValue=4, group=Fake Reflection, parameterName=Reflection Tiling X, sortPriority=1}
  ) {constB=4}
  n113 = Multiply(
   A=ArcsineFast!(None=ComponentMask(None=@n109) {b=true})
   B=Divide(B=MaterialFunctionCall() {materialFunction=&Pi}.Result) {constA=2}
  )
  n114 = Add(A=Multiply(
    A=Multiply(A=@n113) {constB=-1}
    B=ScalarParameter() {defaultValue=2, group=Fake Reflection, parameterName=Reflection Tiling Y, sortPriority=2}
   ) {constB=2}) {constB=0.5}
  n115 = AppendVector(A=@n112, B=@n114)
  n116 = ComponentMask(None=VectorParameter() {group=Fake Reflection, parameterName=Reflection Offset, defaultValue=<rgba 0 0 0 1>}.RGB) {g=true, r=true}
  n117 = NamedRerouteUsage()
  n118 = ScalarParameter() {defaultValue=0.3, group=Fake Reflection, parameterName=Fresnel Distortion Reflection, sliderMax=0.5}
  n119 = Add(A=Frac(None=@n112)) {constB=-0.5}
  n120 = ScalarParameter() {defaultValue=4, group=Fake Reflection, parameterName=Reflection Tiling X, sortPriority=1}
  n121 = AppendVector(A=Multiply(A=Abs(None=@n119), B=Divide(B=@n120)), B=@n114)
  n122 = TextureSampleParameter2D!(
   UVs=LinearInterpolate(A=Add(A=@n115, B=@n116), B=TextureCoordinate(), Alpha=Multiply(A=@n117, B=@n118) {constB=0.3})
   "DDX(UVs)"=DDX(Value=@n121)
   "DDY(UVs)"=DDY(Value=@n121)
   Apply View MipBias
  ) {
   group=Fake Reflection
   mipValueMode=TMVM_Derivative
   parameterName=Fake Reflection Texture
   texture=/SoberBotany/Materials/T_BreakableGlass_Reflection_Hera_Interior
  }
  n123 = Multiply(A=@n122.RGB, B=ScalarParameter() {defaultValue=0.5, group=Fake Reflection, parameterName=Reflection})
  n124 = Multiply(
   A=LinearInterpolate(A=Multiply(A=@n123, B=@n108.RGB), Alpha=NamedRerouteUsage()) {constB=0}
   B=ScalarParameter() {defaultValue=0.8, group=Fake Reflection, parameterName=Reflection Intensity}
  ) {constB=0.5}
  NamedRerouteDeclaration(None=MaterialFunctionCall(Default=StaticSwitchParameter(True=@n107, False=@n124) {parameterName=Opaque?}, Low/Mobile=Constant()) {materialFunction=&MF_LowMobile_PathingSwitch}.Result) {name=Reflection, nodeColor=<rgba 0 1 0.3672 1>}
  n125 = Multiply(A=NamedRerouteUsage(), B=ScalarParameter() {defaultValue=0.34, group=Edge Masks, parameterName=Edge Opacity Gain}) {constB=0.6}
  n126 = ScalarParameter() {defaultValue=1.1, group=Opacity, parameterName=Opacity Max}
  n127 = Saturate(None=Multiply(A=NamedRerouteUsage(), B=NamedRerouteUsage()))
  n128 = NamedRerouteDeclaration(None=Max(A=LinearInterpolate(B=@n126, Alpha=@n127) {constA=0.1}, B=ScalarParameter() {defaultValue=0.7, group=Opacity, parameterName=Min Occlusion})) {name=Opacity, nodeColor=<rgba 0.7578 0 1 1>}
  Reroute(None=StaticSwitchParameter(True=Constant() {r=1}, False=Add(A=@n125, B=@n128)) {parameterName=Opaque?})
  TextureCoordinate()
 ]
}
